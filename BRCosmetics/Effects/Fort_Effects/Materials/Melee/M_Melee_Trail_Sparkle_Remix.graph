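# Translucent unlit, two-sided material for a sparkling melee trail on beam trails and mesh particles.
# Color blends two stripe tints through panning noise and cloud textures, then applies particle color, time-of-day night dimming, an optional dynamic offset and emissive adjustments.
# Alpha combines trail edge masks, near-camera fade and Fresnel with an animated sparkle mask.
# Sparkles come from quantized UV cells pulsing in time-cycled sine patterns, shaped by sphere-masked disco-tile "EQ" bands.
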Material /BRCosmetics/Effects/Fort_Effects/Materials/Melee/M_Melee_Trail_Sparkle_Remix
graph {
 properties {
  bUsedWithBeamTrails=true
  bUsedWithMeshParticles=true
  blendMode=BLEND_Translucent
  pixelDepthOffsetMode=PDOM_Legacy
  shadingModel=MSM_Unlit
  translucencyLightingMode=TLM_VolumetricPerVertexNonDirectional
  twoSided=true
  usedShadingModels=Unlit
 }
 nodes [
  n1 = TextureSample(UVs=Panner(Coordinate=TextureCoordinate() {uTiling=0.2}) {bFractionalPart=true, speedX=-0.65}) {texture=&T_NoisePack_01}
  n2 = ComponentMask(None=TextureCoordinate()) {r=true}
  n3 = LinearInterpolate(Alpha=Add(A=ConstantBiasScale(None=@n1.R) {bias=-0.5, scale=0.2}, B=OneMinus(None=@n2))) {constA=0.01, constB=0.9}
  n4 = Panner(Coordinate=TextureCoordinate() {uTiling=0, vTiling=0.3}) {speedX=-0.2, speedY=-0.2}
  n5 = Panner(Coordinate=TextureCoordinate() {uTiling=0, vTiling=0.2}) {speedX=-0.25, speedY=-0.2}
  n6 = ComponentMask(None=Add(
    A=TextureSample(UVs=@n4) {texture=/Game/Effects/Fort_Effects/PlayerHealthDamage/T_DBNO_Cloud}.RGB
    B=TextureSample(UVs=@n5) {texture=&T_DBNO_Cloud}.RGB
   )) {g=true}
  n7 = ParticleColor()
  n8 = Multiply(
   A=LinearInterpolate(
    A=VectorParameter() {parameterName=Color_Stripes_Top_1, defaultValue=<rgba 1 0.019 0.804 1>}.RGB
    B=VectorParameter() {parameterName=Color_Stripes_Top_2, defaultValue=<rgba 0.5063 0.0724 1 1>}.RGB
    Alpha=Multiply(A=@n3, B=@n6)
   )
   B=@n7.RGB
  )
  n9 = MaterialFunctionCall(Color_Morning=@n8, Color_Day=@n8, Color_Evening=@n8, Color_Night=Multiply(A=@n8) {constB=0.25}) {materialFunction=&Color_Time_of_Day}
  n10 = DynamicParameter() {paramNames [Time,Random,ColorRange,Alpha]}
  MaterialFunctionCall(Emissive=LinearInterpolate(A=@n9.Result, B=Add(A=@n9.Result, B=@n10.ColorRange), Alpha=ScalarParameter() {parameterName=Dynamic Controlled Color Range})) {materialFunction=&MF_FX_EmissiveAdjustments}
  n11 = TextureCoordinate()
  n12 = Saturate(None=Multiply(A=ComponentMask(None=@n11) {r=true}) {constB=3})
  n13 = Constant() {r=0.5}
  n14 = ComponentMask(None=TextureCoordinate()) {g=true}
  n15 = SphereMask(A=TextureCoordinate() {uTiling=0.5}, B=Constant2Vector() {g=0.82, r=0.85}, Radius=Constant() {r=0.8}, Hardness=Constant())
  n16 = Multiply(
   A=SphereMask(
    A=AppendVector(A=@n13, B=@n14)
    B=@n13
    Radius=ScalarParameter() {defaultValue=0.5, parameterName=Mask Radius}
    Hardness=ScalarParameter() {defaultValue=0.9, parameterName=Mask Hardness}
   )
   B=Power(Base=OneMinus(None=@n15)) {constExponent=5}
  )
  n17 = ScalarParameter() {defaultValue=64, parameterName=NearFadeDistance}
  n18 = Multiply(
   A=Multiply(
    A=Multiply(A=@n12, B=@n16)
    B=MaterialFunctionCall(Distance=@n17) {materialFunction=&Near_Camera_Fade}.Result
   )
   B=MaterialFunctionCall(
    Brightness=ScalarParameter() {defaultValue=1.5, parameterName=Fresnel Brightness}
    Exponent=ScalarParameter() {defaultValue=3, parameterName=Fresnel Exponent}
   ) {materialFunction=&MF_Fresnel_FX}.Fresnel Double Sided
  )
  n19 = ComponentMask(None=TextureCoordinate()) {r=true}
  n20 = Multiply(A=Multiply(A=@n19, B=OneMinus(None=@n19))) {constB=2}
  n21 = Multiply(A=Time()) {constB=0.5}
  n22 = Multiply(A=TextureCoordinate() {coordinateIndex=1, uTiling=1.25, vTiling=0.75}, B=Constant() {desc=Macro Iteration, r=2}) {constB=2}
  n23 = MaterialFunctionCall(Float2=Ceil(None=@n22)) {materialFunction=&BreakOutFloat2Components}
  n24 = Multiply(A=@n23.G) {constB=2}
  n25 = Multiply(A=Subtract(A=Add(A=@n23.R, B=@n24)) {constB=3}) {constB=0.3333}
  n26 = Constant() {desc=Micro Iteration, r=4}
  n27 = Ceil(None=Multiply(A=@n22, B=Constant() {desc=Micro Iteration, r=4}) {constB=4})
  n28 = SphereMask(A=@n27, B=Add(A=@n26) {constB=0.5}, Radius=Add(A=@n26)) {attenuationRadius=5, hardnessPercent=0}
  n29 = ScalarParameter() {defaultValue=1, group=Patterns, parameterName=Color Animation Cycle Time}
  n30 = Constant() {r=2}
  n31 = Add(A=Time(), B=Divide(A=Fmod(A=@n27, B=@n30), B=@n30))
  n32 = Multiply(A=Frac(None=Multiply(A=@n29, B=@n31) {constB=0.05, desc=Full cycle 20 second}) {desc="0-1 cycle "}) {constB=3, desc=the number of stages}
  n33 = ScalarParameter() {defaultValue=150, group=Patterns, parameterName=Pattern Transition Speed}
  n34 = MaterialFunctionCall(In=Clamp(None=@n32), Contrast=@n33) {materialFunction=&CheapContrast}
  n35 = Subtract(A=@n32)
  n36 = Clamp(None=@n35)
  n37 = Subtract(A=@n35)
  n38 = LinearInterpolate(
   A=LinearInterpolate(
    A=LinearInterpolate(A=@n25, B=@n28, Alpha=@n34.Result)
    B=Multiply(A=@n27) {constB=0.125}
    Alpha=MaterialFunctionCall(In=@n36, Contrast=@n33) {materialFunction=&CheapContrast}.Result
   )
   B=@n25
   Alpha=MaterialFunctionCall(In=Clamp(None=@n37), Contrast=@n33) {materialFunction=&CheapContrast}.Result
  )
  n39 = Add(A=Sine(None=Add(A=@n21, B=@n38)), B=ScalarParameter() {defaultValue=1, group=Pulse, parameterName=Animation Sine Pulse Low Value}) {constB=1.25}
  n40 = ScalarParameter() {defaultValue=0.1357, group=Pulse, parameterName=Animation Sine Pulse High Value}
  n41 = MaterialFunctionCall(Float2=Ceil(None=@n27)) {materialFunction=&BreakOutFloat2Components}
  n42 = Add(A=@n41.R, B=Multiply(A=@n41.G) {constB=2})
  n43 = Add(
   A=Time()
   B=Multiply(A=Subtract(A=@n42) {constB=3}, B=ScalarParameter() {defaultValue=2.6885, group=Pulse, parameterName=The Magic Number}) {constB=0.05}
  )
  n44 = ScalarParameter() {defaultValue=0.25, group=Pulse, parameterName=Sine Pulse Speed}
  n45 = Add(A=Sine(None=Multiply(A=@n43, B=@n44)), B=ScalarParameter() {defaultValue=5.4534, group=Pulse, parameterName=Sine Pulse Low Value})
  n46 = ScalarParameter() {defaultValue=0.0595, group=Pulse, parameterName=Sine Pulse High Value}
  n47 = Multiply(A=TextureCoordinate(), B=ScalarParameter() {defaultValue=0.25, group=EQ Shape, parameterName=X_Nr_Bands}) {constB=10}
  n48 = ScalarParameter() {defaultValue=0.25, group=EQ Shape, parameterName="Rotation Angle (0-1) (S)", sortPriority=1}
  n49 = ComponentMask(None=TextureSample(UVs=MaterialFunctionCall(UVs=@n47, "Rotation Angle (0-1)"=@n48) {materialFunction=&CustomRotator}.Rotated Values) {samplerType=SAMPLERTYPE_LinearColor, texture=/Game/Weapons/WeaponSkins/Textures/DiscoTiles1_M}.RGB) {g=true, r=true}
  n50 = MaterialFunctionCall(Speed=ScalarParameter() {defaultValue=0.35, group=EQ Shape, parameterName="Speed2 (S)", sortPriority=4}) {materialFunction=&DebugTimeSine}
  n51 = AppendVector(
   A=@n50.Result
   B=MaterialFunctionCall(Speed=ScalarParameter() {defaultValue=0.5, group=EQ Shape, parameterName="Speed (S)", sortPriority=3}) {materialFunction=&DebugTimeSine}.Result
  )
  n52 = LinearInterpolate(A=Constant() {r=0.5}, B=Constant() {r=1}, Alpha=@n50.Result)
  n53 = ScalarParameter() {defaultValue=0.5, group=EQ Shape, parameterName=EQ Shape Hardness, sortPriority=5}
  n54 = TextureCoordinate()
  n55 = ScalarParameter() {defaultValue=0.5, group=EQ Shape, parameterName=Y_Nr_Bands, sortPriority=2}
  n56 = OneMinus(None=Frac(None=Multiply(A=@n54, B=@n55) {constB=10}))
  n57 = ComponentMask(None=Multiply(A=TextureSample(UVs=@n56) {samplerType=SAMPLERTYPE_LinearColor, texture=&DiscoTiles1_M}.RGB) {constB=0.5}) {g=true, r=true}
  n58 = LinearInterpolate(
   A=Constant() {r=0.5}
   B=Constant() {r=1}
   Alpha=MaterialFunctionCall(Speed=ScalarParameter() {defaultValue=0.35, group=EQ Shape, parameterName="Speed2 (S)", sortPriority=4}) {materialFunction=&DebugTimeSine}.Result
  )
  n59 = Multiply(
   A=Multiply(A=Multiply(A=@n39, B=@n40) {constB=0.5}, B=Multiply(A=@n45, B=@n46))
   B=Multiply(A=SphereMask(A=@n49, B=@n51, Radius=@n52, Hardness=@n53), B=SphereMask(A=@n57, B=@n51, Radius=@n58, Hardness=@n53))
  )
  Multiply(A=@n7.A, B=Multiply(A=Multiply(A=@n18, B=@n20), B=Clamp(None=@n59)))
 ]
}
