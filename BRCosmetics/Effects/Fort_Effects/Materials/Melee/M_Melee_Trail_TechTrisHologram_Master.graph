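# Unlit translucent master material for melee beam trails with a techy triangle hologram look.
# A sphere mask, UV gradients and a sine shape a two-color emissive and depth-faded opacity, scaled by squared dynamic Alpha.
# Three rotated, panning curve-atlas stripe sets form the triangle pattern, skipped on mobile along with depth fade.
# "Use Holographic Lines" blends to a noise-eroded UV mask with brighter emissive, and a camera offset shifts the trail.
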Material /BRCosmetics/Effects/Fort_Effects/Materials/Melee/M_Melee_Trail_TechTrisHologram_Master
graph {
 properties {bUsedWithBeamTrails=true, blendMode=BLEND_Translucent, pixelDepthOffsetMode=PDOM_Legacy, shadingModel=MSM_Unlit, usedShadingModels=Unlit}
 nodes [
  MaterialFunctionCall(Offset Amount=ScalarParameter() {defaultValue=5, parameterName=Offset}) {materialFunction=/Engine/Functions/Engine_MaterialFunctions02/WorldPositionOffset/CameraOffset}
  n1 = SphereMask(
   A=WorldPosition() {worldPositionShaderOffset=WPT_ExcludeAllShaderOffsets}.XYZ
   B=ObjectPositionWS()
   Radius=ScalarParameter() {defaultValue=75, parameterName=Radius}
   Hardness=Constant()
  )
  n2 = Multiply(A=@n1, B=@n1)
  n3 = OneMinus(None=ComponentMask(None=TextureCoordinate()) {r=true})
  n4 = TextureCoordinate() {uTiling=0.5, vTiling=0.5}
  n5 = Multiply(A=Power(Base=Min(A=@n2, B=@n3), Exp=ScalarParameter() {defaultValue=3, parameterName=Exp}), B=ComponentMask(None=Sine(None=@n4)) {g=true})
  n6 = ScalarParameter() {parameterName=Use Holographic Lines}
  n7 = TextureCoordinate() {uTiling=2, vTiling=2}
  n8 = OneMinus(None=Abs(None=Add(A=@n7) {constB=-1}))
  n9 = ComponentMask(None=@n8) {r=true}
  n10 = ComponentMask(None=@n8) {g=true}
  n11 = TextureCoordinate() {uTiling=0.05, vTiling=4}
  n12 = Constant2Vector() {g=0.25, r=0.05}
  n13 = Time()
  n14 = TextureSample(UVs=Add(A=Divide(A=@n11) {constB=12}, B=Multiply(A=@n12, B=@n13))) {samplerType=SAMPLERTYPE_Normal, texture=/Game/Global/Textures/Default/Noise/64x64Noise}
  n15 = Saturate(None=Subtract(A=Multiply(A=@n9, B=@n10), B=LinearInterpolate(Alpha=@n14.B) {constA=0.4, constB=0.8}))
  n16 = Multiply(
   A=ParticleColor().A
   B=LinearInterpolate(
    A=VectorParameter() {parameterName=EmissiveB, defaultValue=<rgba 1 0.0415 0 0>}.RGB
    B=VectorParameter() {parameterName=EmissiveA, defaultValue=<rgba 1 0.474 0 0>}.RGB
    Alpha=LinearInterpolate(A=@n5, B=@n15, Alpha=@n6)
   )
  )
  n17 = LinearInterpolate(Alpha=@n15) {constA=1, constB=2}
  MaterialFunctionCall(Emissive=LinearInterpolate(A=@n16, B=Multiply(A=@n16, B=@n17), Alpha=ScalarParameter() {parameterName=Use Holographic Lines})) {materialFunction=&MF_FX_EmissiveAdjustments}
  n18 = DynamicParameter() {paramNames [Alpha,Alpha2,Param3,Param4]}
  n19 = Multiply(A=@n5) {constB=0.25}
  n20 = ShadingPathSwitch(Default=DepthFade(Opacity=@n19) {fadeDistanceDefault=15}, Mobile=@n19) {desc=Migrated from FeatureLevelSwitch}
  n21 = VectorParameter() {parameterName=Curves UV Scale, defaultValue=<rgba 21 3 0 0>}
  n22 = Add(
   A=Multiply(A=TextureCoordinate() {coordinateIndex=1}, B=ComponentMask(None=@n21.RGB) {g=true, r=true})
   B=Multiply(A=Time(), B=ScalarParameter() {defaultValue=0.25, parameterName=Curve Panning Speec}) {constB=0.25}
  )
  n23 = MaterialFunctionCall(UVs=@n22, "Rotation Angle (0-1)"=Constant() {r=0.3333}) {materialFunction=&CustomRotator}
  n24 = CurveAtlasRowParameter(CurveTime=ComponentMask(None=Frac(None=@n23.Rotated Values)) {g=true}) {
   atlas=&AthenaCurveAtlas01
   curve=/Game/Athena/LinearColorCurves/Curve_Masks
   defaultValue=67
   parameterName=CurveMask
  }
  n25 = MaterialFunctionCall(UVs=@n22, "Rotation Angle (0-1)"=Constant() {r=0.1667}) {materialFunction=&CustomRotator}
  n26 = CurveAtlasRowParameter(CurveTime=ComponentMask(None=Frac(None=@n25.Rotated Values)) {g=true}) {
   atlas=&AthenaCurveAtlas01
   curve=&Curve_Masks
   defaultValue=67
   parameterName=CurveMask
  }
  n27 = CurveAtlasRowParameter(CurveTime=ComponentMask(None=Frac(None=@n22)) {g=true}) {
   atlas=&AthenaCurveAtlas01
   curve=&Curve_Masks
   defaultValue=67
   parameterName=CurveMask
  }
  n28 = Add(A=Add(A=@n24.G, B=@n26.G), B=@n27.G)
  n29 = ShadingPathSwitch(
   Default=LinearInterpolate(A=ScalarParameter() {defaultValue=-2, parameterName=Triangle}, B=Constant() {r=1}, Alpha=Saturate(None=@n28))
   Mobile=Constant() {r=1}
  ) {desc=Migrated from FeatureLevelSwitch}
  Multiply(
   A=Multiply(A=@n18.Alpha, B=@n18.Alpha)
   B=LinearInterpolate(A=Multiply(A=@n20, B=@n29), B=@n15, Alpha=ScalarParameter() {parameterName=Use Holographic Lines})
  )
 ]
}
